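# Anim Blueprint for the face accessory of the male IvoryDwell AcesHood cosmetic, posing its hood bones.
# AnimGraph copies the attached parent's pose, chains Modify Bone nodes over kin_hood bones, then simulates dyn_hoodTop with head-relative AnimDynamics.
# Customization load caches the body mesh; each update checks it has kin_hoodSide_l, then stores its hood and lace bone transforms in parent-bone space.
# Lace values have no matching AnimGraph nodes.
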AnimBlueprint /BRCosmetics/Characters/Player/Male/Medium/Bodies/M_MED_IvoryDwell_AcesHood/Meshes/Parts/M_MED_IvoryDwell_AcesHood_FaceAcc_AnimBP
target_skeleton=&/M_MED_IvoryDwell_AcesHood_FaceAcc_Skeleton
graphs [
 AnimGraph {nodes [
   n1 = "Transform (Modify) Bone - Bone: kin_hoodSide_l"!(
    ComponentPose=Local To Component(LocalPose=Copy Pose From Mesh() {bUseAttachedParent=true})
    bAlphaBoolEnabled:bool=True
    Translation="0,0,0"
    Rotation="0,0,0"
   ) -> (Pose) {
    alphaInputType=Bool
    rotationMode=BMM_Replace
    rotationSpace=BCS_ParentBoneSpace
    translationMode=BMM_Replace
    translationSpace=BCS_ParentBoneSpace
    alphaBoolBlend {blendOption=Linear}
    boneToModify {boneName=kin_hoodSide_l}
   }
   n2 = "Transform (Modify) Bone - Bone: kin_hoodSide_r"!(ComponentPose=@n1, bAlphaBoolEnabled:bool=True, Translation="0,0,0", Rotation="0,0,0") -> (Pose) {
    alphaInputType=Bool
    rotationMode=BMM_Replace
    rotationSpace=BCS_ParentBoneSpace
    translationMode=BMM_Replace
    translationSpace=BCS_ParentBoneSpace
    alphaBoolBlend {blendOption=Linear}
    boneToModify {boneName=kin_hoodSide_r}
   }
   n3 = "Transform (Modify) Bone - Bone: kin_hood_def_1_c"!(ComponentPose=@n2, bAlphaBoolEnabled:bool=True, Translation="0,0,0", Rotation="0,0,0") -> (Pose) {
    alphaInputType=Bool
    rotationMode=BMM_Replace
    rotationSpace=BCS_ParentBoneSpace
    translationMode=BMM_Replace
    translationSpace=BCS_ParentBoneSpace
    alphaBoolBlend {blendOption=Linear}
    boneToModify {boneName=kin_hood_def_1_c}
   }
   n4 = "Transform (Modify) Bone - Bone: kin_hood_def_1_l"!(ComponentPose=@n3, bAlphaBoolEnabled:bool=True, Translation="0,0,0", Rotation="0,0,0") -> (Pose) {
    alphaInputType=Bool
    rotationMode=BMM_Replace
    rotationSpace=BCS_ParentBoneSpace
    translationMode=BMM_Replace
    translationSpace=BCS_ParentBoneSpace
    alphaBoolBlend {blendOption=Linear}
    boneToModify {boneName=kin_hood_def_1_l}
   }
   n5 = "Transform (Modify) Bone - Bone: kin_hood_def_1_r"!(ComponentPose=@n4, bAlphaBoolEnabled:bool=True, Translation="0,0,0", Rotation="0,0,0") -> (Pose) {
    alphaInputType=Bool
    rotationMode=BMM_Replace
    rotationSpace=BCS_ParentBoneSpace
    translationMode=BMM_Replace
    translationSpace=BCS_ParentBoneSpace
    alphaBoolBlend {blendOption=Linear}
    boneToModify {boneName=kin_hood_def_1_r}
   }
   n6 = "Transform (Modify) Bone - Bone: kin_hood_def_2"!(ComponentPose=@n5, bAlphaBoolEnabled:bool=True, Translation="0,0,0", Rotation="0,0,0") -> (Pose) {
    alphaInputType=Bool
    rotationMode=BMM_Replace
    rotationSpace=BCS_ParentBoneSpace
    translationMode=BMM_Replace
    translationSpace=BCS_ParentBoneSpace
    alphaBoolBlend {blendOption=Linear}
    boneToModify {boneName=kin_hood_def_2}
   }
   n7 = "Transform (Modify) Bone - Bone: kin_hood_def_2_l"!(ComponentPose=@n6, bAlphaBoolEnabled:bool=True, Translation="0,0,0", Rotation="0,0,0") -> (Pose) {
    alphaInputType=Bool
    rotationMode=BMM_Replace
    rotationSpace=BCS_ParentBoneSpace
    translationMode=BMM_Replace
    translationSpace=BCS_ParentBoneSpace
    alphaBoolBlend {blendOption=Linear}
    boneToModify {boneName=kin_hood_def_2_l}
   }
   n8 = "Transform (Modify) Bone - Bone: kin_hood_def_2_r"!(ComponentPose=@n7, bAlphaBoolEnabled:bool=True, Translation="0,0,0", Rotation="0,0,0", Scale="1,1,1") -> (Pose) {
    alphaInputType=Bool
    rotationMode=BMM_Replace
    rotationSpace=BCS_ParentBoneSpace
    scaleMode=BMM_Additive
    scaleSpace=BCS_BoneSpace
    translationMode=BMM_Replace
    translationSpace=BCS_ParentBoneSpace
    alphaBoolBlend {blendOption=Linear}
    boneToModify {boneName=kin_hood_def_2_r}
   }
   n9 = Anim Dynamics - Bone: dyn_hoodTop!(ComponentPose=@n8, Alpha=1) -> (Pose) {
    angularDampingOverride=1
    angularSpringConstant=5000
    bAngularSpring=true
    bOverrideAngularDamping=true
    bUseGravityOverride=true
    gravityScale=0.35
    simulationSpace=BoneRelative
    boundBone {boneName=dyn_hoodTop}
    componentAppliedLinearAccClamp=<3000 3000 3000>
    componentLinearAccScale {z=1}
    componentLinearVelScale {z=1}
    gravityOverride {z=-980}
    physicsBodyDefinitions [{
      boundBone {boneName=dyn_hoodTop}
      boxExtents=<20 20 20>
      constraintSetup {angularLimitsMax=<0 0 0>, angularLimitsMin=<0 -10 0>, angularTarget=<1 0 0>}
      localJointOffset=<0.2 0 0>
     }]
    relativeSpaceBone {boneName=head}
    simSpaceSettings {maxAngularAcceleration=1, maxAngularVelocity=6, simSpaceAngularAlpha=1}
   }
   Output Pose(Result=Component To Local(ComponentPose=@n9)) {layerGroup=None}
  ]}
 EventGraph {nodes [
   Event BlueprintUpdateAnimation()
   n1 = Get BodyMesh() -> (BodyMesh:Skeletal Mesh Component Budgeted Object Reference)
   > Is Valid(InputObject=@n1)
   n2 = GetSocketBoneName!(self:Skinned Mesh Component Object Reference=@n1, InSocketName:Name=kin_hoodSide_l) -> (ReturnValue:Name=None)
   > n3 = Set isUsingItsBody(isUsingItsBody:bool="Not Equal Exactly (String)"(A="To String (Name)"(InName=@n2), B=None)) -> (Output_Get:bool)
   > Branch(Condition=@n3)
   n4 = Get BodyMesh()
   n5 = BreakTransform(InTransform=GetSocketTransform(self=@n4, InSocketName=kin_hoodSide_l, TransformSpace=RTS_ParentBoneSpace))
   > Set kin_hoodSide_l_Translation(kin_hoodSide_l_Translation:Vector=@n5) -> (Output_Get:Vector)
   > Set kin_hoodSide_l_Rotation(kin_hoodSide_l_Rotation:Rotator=@n5.Rotation) -> (Output_Get:Rotator)
   n6 = BreakTransform(InTransform=GetSocketTransform(self=@n4, InSocketName=kin_hoodSide_r, TransformSpace=RTS_ParentBoneSpace))
   > Set kin_hoodSide_r_Translation(kin_hoodSide_r_Translation:Vector=@n6) -> (Output_Get:Vector)
   > Set kin_hoodSide_r_Rotation(kin_hoodSide_r_Rotation:Rotator=@n6.Rotation) -> (Output_Get:Rotator)
   n7 = BreakTransform(InTransform=GetSocketTransform(self=@n4, InSocketName=kin_hood_def_1_c, TransformSpace=RTS_ParentBoneSpace))
   > Set kin_hood_def_1_c_Translation(kin_hood_def_1_c_Translation:Vector=@n7) -> (Output_Get:Vector)
   > Set kin_hood_def_1_c_Rotation(kin_hood_def_1_c_Rotation:Rotator=@n7.Rotation) -> (Output_Get:Rotator)
   n8 = BreakTransform(InTransform=GetSocketTransform(self=@n4, InSocketName=kin_hood_def_1_l, TransformSpace=RTS_ParentBoneSpace))
   > Set kin_hood_def_1_l_Translation(kin_hood_def_1_l_Translation:Vector=@n8) -> (Output_Get:Vector)
   > Set kin_hood_def_1_l_Rotation(kin_hood_def_1_l_Rotation:Rotator=@n8.Rotation) -> (Output_Get:Rotator)
   n9 = BreakTransform(InTransform=GetSocketTransform(self=@n4, InSocketName=kin_hood_def_1_r, TransformSpace=RTS_ParentBoneSpace))
   > Set kin_hood_def_1_r_Translation(kin_hood_def_1_r_Translation:Vector=@n9) -> (Output_Get:Vector)
   > Set kin_hood_def_1_r_Rotation(kin_hood_def_1_r_Rotation:Rotator=@n9.Rotation) -> (Output_Get:Rotator)
   n10 = BreakTransform(InTransform=GetSocketTransform(self=@n4, InSocketName=kin_hood_def_2, TransformSpace=RTS_ParentBoneSpace))
   > Set kin_hood_def_2_Translation(kin_hood_def_2_Translation:Vector=@n10) -> (Output_Get:Vector)
   > Set kin_hood_def_2_Rotation(kin_hood_def_2_Rotation:Rotator=@n10.Rotation) -> (Output_Get:Rotator)
   n11 = BreakTransform(InTransform=GetSocketTransform(self=@n4, InSocketName=kin_hood_def_2_l, TransformSpace=RTS_ParentBoneSpace))
   > Set kin_hood_def_2_l_Translation(kin_hood_def_2_l_Translation:Vector=@n11) -> (Output_Get:Vector)
   > Set kin_hood_def_2_l_Rotation(kin_hood_def_2_l_Rotation:Rotator=@n11.Rotation) -> (Output_Get:Rotator)
   n12 = BreakTransform(InTransform=GetSocketTransform(self=@n4, InSocketName=kin_hood_def_2_r, TransformSpace=RTS_ParentBoneSpace))
   > Set kin_hood_def_2_r_Translation(kin_hood_def_2_r_Translation:Vector=@n12) -> (Output_Get:Vector)
   > Set kin_hood_def_2_r_Rotation(kin_hood_def_2_r_Rotation:Rotator=@n12.Rotation) -> (Output_Get:Rotator)
   > Set kin_hood_def_2_r_Scale(kin_hood_def_2_r_Scale:Vector=@n12.Scale) -> (Output_Get:Vector)
   n13 = BreakTransform(InTransform=GetSocketTransform(self=@n4, InSocketName=base_lace_1_l, TransformSpace=RTS_ParentBoneSpace))
   > Set base_lace_1_l_Translation(base_lace_1_l_Translation:Vector=@n13) -> (Output_Get:Vector)
   > Set base_lace_1_l_Rotation(base_lace_1_l_Rotation:Rotator=@n13.Rotation) -> (Output_Get:Rotator)
   n14 = BreakTransform(InTransform=GetSocketTransform(self=@n4, InSocketName=base_lace_1_r, TransformSpace=RTS_ParentBoneSpace))
   > Set base_lace_1_r_Translation(base_lace_1_r_Translation:Vector=@n14) -> (Output_Get:Vector)
   > Set base_lace_1_r_Rotation(base_lace_1_r_Rotation:Rotator=@n14.Rotation) -> (Output_Get:Rotator)
   n15 = Event OnCharacterCustomizationLoaded()
   > Set BodyMesh(BodyMesh:Skeletal Mesh Component Budgeted Object Reference=Get Array Item with EFort Custom Part Type(Array=@n15.CustomMeshParts)) -> (Output_Get:Skeletal Mesh Component Budgeted Object Reference)
  ]}
]
variables [
 base_lace_1_l_Rotation {type=Rotator}
 base_lace_1_l_Translation {type=Vector}
 base_lace_1_r_Rotation {type=Rotator}
 base_lace_1_r_Translation {type=Vector}
 BodyMesh {type=SkeletalMeshComponentBudgeted}
 isUsingItsBody {type=boolean}
 kin_hood_def_1_c_Rotation {type=Rotator}
 kin_hood_def_1_c_Translation {type=Vector}
 kin_hood_def_1_l_Rotation {type=Rotator}
 kin_hood_def_1_l_Translation {type=Vector}
 kin_hood_def_1_r_Rotation {type=Rotator}
 kin_hood_def_1_r_Translation {type=Vector}
 kin_hood_def_2_l_Rotation {type=Rotator}
 kin_hood_def_2_l_Translation {type=Vector}
 kin_hood_def_2_r_Rotation {type=Rotator}
 kin_hood_def_2_r_Scale {type=Vector}
 kin_hood_def_2_r_Translation {type=Vector}
 kin_hood_def_2_Rotation {type=Rotator}
 kin_hood_def_2_Translation {type=Vector}
 kin_hoodSide_l_Rotation {type=Rotator}
 kin_hoodSide_l_Translation {type=Vector}
 kin_hoodSide_r_Rotation {type=Rotator}
 kin_hoodSide_r_Translation {type=Vector}
]
